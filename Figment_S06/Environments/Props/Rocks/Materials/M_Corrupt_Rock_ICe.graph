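# Nanite-enabled masked rock material that gives corrupt rocks an icy tint, snow on upward-facing surfaces and a water-colored lower portion.
# A world-normal mask blends tinted rock into tiled white-rock snow, then a world-height gradient blends that into a time-of-day water color and flattens the normal map.
# Rocks slowly spin about vertical, with speed and direction derived from object position.
# World-space normals are rotation-corrected to match.
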Material /Figment_S06/Environments/Props/Rocks/Materials/M_Corrupt_Rock_ICe
graph {
 properties {bTangentSpaceNormal=false, bUsedWithNanite=true, blendMode=BLEND_Masked, pixelDepthOffsetMode=PDOM_Legacy, usedShadingModels=Default Lit}
 nodes [
  n1 = VectorParameter() {parameterName=Morning, defaultValue=<rgba 0.1235 0.1422 0.1927 1>}
  n2 = VectorParameter() {parameterName=Day, defaultValue=<rgba 0.0384 0.1056 0.1615 1>}
  n3 = VectorParameter() {parameterName=Evening, defaultValue=<rgba 0.0648 0.1002 0.131 1>}
  n4 = VectorParameter() {parameterName=Night, defaultValue=<rgba 0.0364 0.0979 0.15 1>}
  n5 = TextureSample() {texture=/Game/Environments/Props/Rocks/Textures/T_Corrupt_Rock_AB_D}
  n6 = VectorParameter() {parameterName=RockColor, defaultValue=<rgba 2.074 5.4474 8 0>}
  n7 = Add(A=TextureSample(UVs=TextureCoordinate() {uTiling=16, vTiling=16}) {texture=/Game/Packages/DS_Fortnight/WhiteRockCustom_D}.RGB) {constB=0.33}
  n8 = VectorParameter() {parameterName=SnowColor, defaultValue=<rgba 0.7344 0.8869 1 0>}
  n9 = DotProduct(A=PixelNormalWS(), B=Constant3Vector() {constant=<rgba 0 0 1 0>})
  n10 = ScalarParameter() {defaultValue=4, parameterName=SnowMaskPower}
  n11 = Clamp(None=Multiply(A=Power(Base=@n9, Exp=@n10), B=ScalarParameter() {defaultValue=1, parameterName=SnowOpacity}))
  n12 = ComponentMask(None=WorldPosition().XYZ) {b=true}
  n13 = ScalarParameter() {defaultValue=5000, parameterName=WaterHeightOffset}
  n14 = Power(
   Base=Divide(A=Add(A=@n12, B=@n13), B=ScalarParameter() {defaultValue=500, parameterName=WaterHeightBlend})
   Exp=ScalarParameter() {defaultValue=1, parameterName=WaterBlendPower}
  )
  n15 = Clamp(None=@n14)
  LinearInterpolate(
   A=Multiply(
    A=MaterialFunctionCall(Color_Morning=@n1.RGB, Color_Day=@n2.RGB, Color_Evening=@n3.RGB, Color_Night=@n4.RGB) {materialFunction=&Color_Time_of_Day}.Result
    B=VectorParameter() {parameterName=WaterColor, defaultValue=<rgba 0.2344 0.6214 1 1>}.RGB
   )
   B=LinearInterpolate(A=Multiply(A=@n5.RGB, B=@n6.RGB), B=Multiply(A=@n7, B=@n8.RGB), Alpha=@n11)
   Alpha=@n15
  )
  LinearInterpolate(Alpha=@n11) {constA=0.4, constB=0}
  LinearInterpolate(Alpha=Clamp(None=Subtract(A=@n11, B=@n15))) {constA=0.08, constB=0.6}
  n16 = Constant3Vector() {constant=<rgba 0 0 1 0>}
  n17 = ObjectPositionWS()
  n18 = Frac(None=Divide(A=ComponentMask(None=@n17) {r=true}, B=ComponentMask(None=@n17) {g=true}))
  n19 = Multiply(A=Multiply(A=Time(), B=LinearInterpolate(Alpha=@n18) {constA=-1}), B=ScalarParameter() {defaultValue=0.01, parameterName=SpinSpeed})
  n20 = Constant3Vector() {constant=<rgba 0 0 1 0>}
  n21 = TextureSample() {samplerType=SAMPLERTYPE_Normal, texture=/Game/Environments/Props/Rocks/Textures/T_Corrupt_Rock_AB_N}
  n22 = Clamp(None=@n14) {maxDefault=0.8}
  MaterialFunctionCall(
   Rotation Axis=@n16
   Rotation Angle=@n19
   World Space Vertex Normals=MaterialFunctionCall(BaseNormal=VertexNormalWS(), AdditionalNormal=LinearInterpolate(A=@n20, B=@n21.RGB, Alpha=@n22)) {materialFunction=&BlendAngleCorrectedNormals}.Result
  ) {materialFunction=&FixRotateAboutAxisNormals}
  Reroute(None=RotateAboutAxis(NormalizedRotationAxis=@n16, RotationAngle=@n19, PivotPoint=ObjectPositionWS(), Position=WorldPosition().XYZ))
 ]
}
